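# MetaSound preset that supplies the idle music loop for the Duck Coast music encounter in Wild Estate.
# It derives from the volume-based, clockless MusicEncounter loop parent and sets the wave to the Phase 0 combat loop.
# The parent exposes distance attenuation inputs, fade-in time and pausing with a timed fade-out.
# Output is stereo, routed to the NPC music sound class.
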MetaSoundSource /WildEstate_DuckCoast/Audio/IP_MSS_DuckCoast_Music_Idle
preset_parent=/MusicEncounter/MetaSounds/MSS_MusicEncounter_VolumeLoop_NoClock
inputs [
 FadeInTime {type=Time}
 MasterVolume {type=Float}
 Music.WaveAsset {type=WaveAsset}
 MusicEncounter.Attenuation.Distance {type=Float}
 MusicEncounter.Attenuation.FalloffDistance {type=Float}
 MusicEncounter.Attenuation.InnerRadius {type=Float}
 MusicEncounter.bIsPaused {type=Bool}
 MusicEncounter.PauseFadeOutTime {type=Time}
 UE.Source.OnPlay {type=Trigger}
]
interfaces [UE.OutputFormat.Stereo:1.0, UE.Source.OneShot:1.0, UE.Source:1.1]
outputs [
 "UE.OutputFormat.Stereo.Audio:0" {type=Audio}
 "UE.OutputFormat.Stereo.Audio:1" {type=Audio}
 UE.Source.OneShot.OnFinished {type=Trigger}
]
pages [Default {input_defaults {
   FadeInTime=0
   MasterVolume=1
   Music.WaveAsset=/WildEstate_DuckCoast/Audio/SoundWaves/MX_S38_Combat_IP_DuckCoast_Phase_0_LP
   MusicEncounter.Attenuation.Distance=0
   MusicEncounter.Attenuation.FalloffDistance=3000
   MusicEncounter.Attenuation.InnerRadius=3000
   MusicEncounter.PauseFadeOutTime=2
   MusicEncounter.bIsPaused=null
   UE.Source.OnPlay=null
  }}]
properties {
 assetImportData=&self:AssetImportData
 numChannels=2
 outputFormat=Stereo
 sampleRate=48000
 soundClassObject=&Fort_Music_NPC
 assetUserData [&self:AudioAssetUserData_0]
}
